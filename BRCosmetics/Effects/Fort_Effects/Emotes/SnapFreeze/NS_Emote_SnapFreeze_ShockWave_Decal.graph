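# Niagara system for the SnapFreeze emote's shockwave "decal", drawing purple HDR glow sprites over a character's body.
# One emitter bursts twelve camera-offset particles through two sprite renderers, each given a spawn group that indexes one of twelve filtered skeletal bones.
# Particles follow those skinned bones every update, while a curve ramps alpha to a brief late-life spike then fades it out.
# A time-of-day scalar brightens the color at night.
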NiagaraSystem /BRCosmetics/Effects/Fort_Effects/Emotes/SnapFreeze/NS_Emote_SnapFreeze_ShockWave_Decal
dependencies {
 usedDataInterfaces [&NiagaraDataInterfaceSkeletalMesh, &NiagaraDataInterfaceCurve]
 usedDynamicInputs []
 usedModules [&SystemState]
}
emitters {ParticleDecale001 {
  emitterUpdateScript [
   &EmitterState(Inactive Response=::NewEnumerator1, Life Cycle Mode=::NewEnumerator1, Loop Behavior=::NewEnumerator1, Loop Duration=2, UseLoopDelay=false)
   &SpawnBurst_Instantaneous(Spawn Count=12, Use Spawn Probability=false, UseLoopCountLimit=false)
  ]
  particleSpawnScript [
   &FN_SetSpawnGroup(Number of Spawn Groups=12)
   &InitializeParticle(Color=<rgba 2.2 0 6 1>, Sprite Size Mode=::NewEnumerator3, Uniform Sprite Size=50, UsePositionOffset=false, Write Material Random=true)
   &CameraOffset(Camera Offset Amount=30)
  ]
  particleUpdateScript [
   &ParticleState()
   &SkeletalMeshLocation(
    Bone / Socket Index=$Particles.SpawnGroup
    Bone Sampling Mode=::NewEnumerator4
    MaskSpawnGroup=false
    Skeletal Mesh=DataInterface() {propertyValues {
      DefaultMesh=null
      ExcludeBoneName=root
      OptionalComponentClass=null
      PreviewMesh=/BRCosmetics/Characters/Player/Female/Medium/Bodies/F_MED_SnapFreeze_Graze/Meshes/F_MED_SnapFreeze_Graze
      SkinningMode=SkinOnTheFly
      SoftSourceActor=null
      SourceMode=Default
      UvSetIndex=0
      WholeMeshLOD=-1
      bExcludeBone=false
      bReadDeformedGeometry=true
      bRequireCurrentFrameData=true
      ComponentTags []
      FilteredBones [pelvis,upperarm_l,upperarm_r,lowerarm_l,lowerarm_r,hand_l,hand_r,calf_l,calf_r,foot_l,foot_r,head]
      FilteredSockets []
      MeshUserParameter {parameter {name=None, typeDefHandle {registeredTypeIndex=1}, varData []}}
      SamplingRegions []
     }}
   )
   &ScaleSpriteSize(Uniform Curve Sprite Scale=DataInterface() {propertyValues {Curve {keys [{tangentMode=RCTM_User, time=0, value=0.6}, {tangentMode=RCTM_User, time=1, value=1}]}}}) {enabled=false}
   &ScaleColor(
    Scale Alpha=&FloatFromCurve(FloatCurve=DataInterface() {propertyValues {Curve {keys [
         <at 0 0>
         <at 0.3 0.25>
         <at 0.4635 0.3>
         {arriveTangent=-6.0371, interpMode=RCIM_Cubic, leaveTangent=0.0961, tangentMode=RCTM_Break, time=0.55, value=1.2}
         {arriveTangent=0.008, interpMode=RCIM_Cubic, leaveTangent=-12.1745, tangentMode=RCTM_Break, time=0.66, value=1.2}
         {arriveTangent=-3.5294, interpMode=RCIM_Cubic, leaveTangent=-3.5294, time=0.75, value=0.4}
         <at 1 0 RCIM_Cubic>
        ]}}})
    Scale RGB=&VectorFromFloat(Value=&FNTimeOfDayScalar(Night Scale=1.5, Morning Scale=1, Evening Scale=1))
    ScaleA=true
    ScaleRGB=true
   )
  ]
  rendererClasses [&NiagaraSpriteRendererProperties]
  renderers [
   {rendererClass=&NiagaraSpriteRendererProperties}
   {rendererClass=&NiagaraSpriteRendererProperties, rendererIndex=1}
  ]
 }}
summary {
 systemName=NS_Emote_SnapFreeze_ShockWave_Decal
 emitters [{emitterName=ParticleDecale001, rendererClasses [&NiagaraSpriteRendererProperties]}]
 userVariables [
  User.bIsFrontEnd {type=NiagaraBool}
  User.bIsFrontEndPreview {type=NiagaraBool}
 ]
}
