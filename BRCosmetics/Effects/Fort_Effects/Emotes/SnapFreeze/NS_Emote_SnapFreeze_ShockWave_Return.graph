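# Niagara system for the SnapFreeze emote's shockwave return, a one-shot purple energy effect that collapses onto the character and flares out.
# Shrinking ring, wave, and inward-moving sparks and streaks lead into a delayed expanding color wave, light-casting glow and downward sweep.
# A refraction mesh contracts and re-expands, while glow sprites track twelve skeletal bones via spawn groups.
# Emitters run independently on timed curves, with various time-of-day, front-end and platform scaling.
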NiagaraSystem /BRCosmetics/Effects/Fort_Effects/Emotes/SnapFreeze/NS_Emote_SnapFreeze_ShockWave_Return
dependencies {
 usedDataInterfaces [
  &NiagaraDataInterfaceCurve
  &NiagaraDataInterfaceColorCurve
  &NiagaraDataInterfaceArrayFloat
  &NiagaraDataInterfaceSkeletalMesh
  &NiagaraDataInterfaceMeshRendererInfo
  &NiagaraDataInterfacePlatformSet
 ]
 usedDynamicInputs [&RandomRangeInt, &Multiply_Vector2D_ByFloat]
 usedModules [
  &self:SystemSpawnScript.SystemScriptSource.SystemScriptGraph.NiagaraNodeAssignment_0.SetVariables_#1
  &SystemState
 ]
}
emitters {
 Chest_Glow {
  emitterUpdateScript [
   &EmitterState(
    Inactive Response=::NewEnumerator1
    Life Cycle Mode=::NewEnumerator1
    Loop Behavior=::NewEnumerator1
    Loop Delay=1.6
    Loop Duration=2
    UseLoopDelay=false
   )
   &SpawnBurst_Instantaneous(Use Spawn Probability=false, UseLoopCountLimit=false)
  ]
  particleSpawnScript [
   &InitializeParticle(
    Color=<rgba 0.839 0.1301 2.5 1>
    Lifetime=1.1
    Position Offset=<0 15 0>
    Sprite Rotation Mode=::NewEnumerator2
    Sprite Size Mode=::NewEnumerator3
    Uniform Sprite Size=80
    UsePositionOffset=true
    Write Material Random=true
   )
   &CameraOffset()
  ]
  particleUpdateScript [
   &ParticleState()
   &ScaleSpriteSize(
    Uniform Curve Scale=&FN_FrontEndFloatModifier(Gameplay=1, Front End=0.8)
    Uniform Curve Sprite Scale=DataInterface() {propertyValues {Curve {keys [
        {arriveTangent=0.7159, interpMode=RCIM_Cubic, leaveTangent=0.7159, tangentMode=RCTM_User, time=0, value=0.75}
        {arriveTangent=0.005, interpMode=RCIM_Cubic, leaveTangent=0.005, tangentMode=RCTM_User, time=1, value=1.05}
       ]}}}
   )
   &ScaleColor(
    Scale Alpha=&FloatFromCurve(
     FloatCurve=DataInterface() {propertyValues {Curve {keys [
         {arriveTangent=0.8482, interpMode=RCIM_Cubic, leaveTangent=0.8482, tangentMode=RCTM_User, time=0, value=0}
         {arriveTangent=1.7409, interpMode=RCIM_Cubic, leaveTangent=-6.3421, tangentMode=RCTM_Break, time=0.8, value=1}
         {arriveTangent=-0.0028, interpMode=RCIM_Cubic, leaveTangent=-3.936, tangentMode=RCTM_Break, time=1, value=0}
        ]}}}
     Scale Curve=&Multiply_Float(A=1, B=&FNTimeOfDayScalar(Night Scale=0.7, Morning Scale=0.9, Evening Scale=0.9))
    )
    ScaleA=true
    ScaleRGB=true
   )
  ]
  rendererClasses [&NiagaraSpriteRendererProperties]
  renderers [{rendererClass=&NiagaraSpriteRendererProperties}]
 }
 Color_wave {
  emitterUpdateScript [
   &EmitterState(
    Inactive Response=::NewEnumerator1
    Life Cycle Mode=::NewEnumerator1
    Loop Behavior=::NewEnumerator1
    Loop Delay=0.4
    Loop Duration=2
    UseLoopDelay=true
   )
   &SpawnBurst_Instantaneous(Use Spawn Probability=false, UseLoopCountLimit=false)
  ]
  particleSpawnScript [&InitializeParticle(
    Color=<rgba 3 3 3 2>
    Lifetime=0.8
    Mesh Scale Mode=::NewEnumerator3
    Mesh Uniform Scale=-3.5
    Position Offset=$System.Position
    Sprite Rotation Angle=90
    Sprite Rotation Mode=::NewEnumerator0
    Sprite Size Mode=::NewEnumerator3
    Uniform Sprite Size=150
    UsePositionOffset=true
    Write Material Random=true
   )]
  particleUpdateScript [
   &ParticleState()
   &ScaleSpriteSize(
    Uniform Curve Scale=1.2
    Uniform Curve Sprite Scale=DataInterface() {propertyValues {Curve {keys [
        {arriveTangent=6.5767, interpMode=RCIM_Cubic, leaveTangent=6.5767, tangentMode=RCTM_User, time=0, value=0}
        {arriveTangent=0.7192, interpMode=RCIM_Cubic, leaveTangent=0.7192, tangentMode=RCTM_User, time=0.2305, value=0.7836}
        {tangentMode=RCTM_User, time=1, value=1}
       ]}}}
   )
   &DynamicMaterialParameters(
    Index 0 Param 1=1
    Index 0 Param 2=&FloatFromCurve(FloatCurve=DataInterface() {propertyValues {Curve {keys [
         {interpMode=RCIM_Cubic, tangentMode=RCTM_User, time=0, value=1}
         {arriveTangent=-1, interpMode=RCIM_Cubic, leaveTangent=-1, time=0.6, value=0.7}
         {interpMode=RCIM_Cubic, leaveTangent=-3.3753, time=1, value=0}
        ]}}})
    Index 0 Param 3=1
    Index 0 Param 4=1
    Module.Param0WriteEnabled=true
    Module.Param1WriteEnabled=true
    Module.Param2WriteEnabled=true
    Module.Param3WriteEnabled=true
   )
   &ScaleColor(
    Scale Alpha=&FloatFromCurve(
     FloatCurve=DataInterface() {propertyValues {Curve {keys [
         {arriveTangent=-2.878, interpMode=RCIM_Cubic, leaveTangent=-2.878, tangentMode=RCTM_User, time=0, value=1}
         {arriveTangent=-1, interpMode=RCIM_Cubic, leaveTangent=-1, time=0.3, value=0.3}
         {interpMode=RCIM_Cubic, tangentMode=RCTM_User, time=1, value=0}
        ]}}}
     Scale Curve=&Multiply_Float(A=2, B=&FNTimeOfDayScalar(Night Scale=0.5, Morning Scale=0.8, Evening Scale=0.8))
    )
    Scale RGB=&MakeVectorFromLinearColorRGB(Color=&MultiplyLinearColor(
      Linear Color=&ColorFromCurve(ColorCurve=DataInterface() {propertyValues {
         AlphaCurve {keys []}
         BlueCurve {keys [<at 0.1259 2 RCIM_Cubic>, <at 0.2277 5>, {interpMode=RCIM_Cubic, leaveTangent=-0.7175, time=0.4478, value=2}]}
         GreenCurve {keys [<at 0.1259 0.1681 RCIM_Cubic>, <at 0.2277 0.2344>, {interpMode=RCIM_Cubic, leaveTangent=-4.4353, time=0.4478, value=0.1399}]}
         RedCurve {keys [<at 0.1259 0.5589 RCIM_Cubic>, <at 0.2277 1.1257>, {interpMode=RCIM_Cubic, leaveTangent=2.2176, time=0.4478, value=0.5865}]}
        }})
      Scale Factor=<1 1 1 1>
     ))
    ScaleA=true
    ScaleRGB=true
   )
   /SetVariables(Particles.Color02:LinearColor=&MakeLinearColorFromVectorAndFloat(
     "Vector (RGB)"=&MakeVectorFromLinearColorRGB(Color=&MultiplyLinearColor(Linear Color=<rgba 1.7071 0.6284 5 1>, Scale Factor=<1 1 1 1>))
     "Float (Alpha)"=&FloatFromCurve(FloatCurve=DataInterface() {propertyValues {Curve {keys [
          {arriveTangent=-1.6399, interpMode=RCIM_Cubic, leaveTangent=-1.6399, tangentMode=RCTM_User, time=0, value=1}
          {interpMode=RCIM_Cubic, tangentMode=RCTM_User, time=1, value=0}
         ]}}})
    )) {isSetParameters=true, module=SetVariables_#2}
  ]
  rendererClasses [&NiagaraSpriteRendererProperties]
  renderers [
   {rendererClass=&NiagaraSpriteRendererProperties}
   {rendererClass=&NiagaraSpriteRendererProperties, rendererIndex=1}
  ]
 }
 Energy_Ring {
  emitterUpdateScript [
   &EmitterState(Inactive Response=::NewEnumerator1, Life Cycle Mode=::NewEnumerator1, Loop Behavior=::NewEnumerator1, Loop Duration=2, UseLoopDelay=false)
   &SpawnBurst_Instantaneous(Use Spawn Probability=false, UseLoopCountLimit=false)
  ]
  particleSpawnScript [
   &InitializeParticle(
    Lifetime=0.55
    Mesh Scale Mode=::NewEnumerator3
    Mesh Uniform Scale=-3.5
    Position Offset=$System.Position
    UsePositionOffset=true
    Write Material Random=true
   )
   &InitialMeshOrientation(
    Mesh Orientation Mode=::NewEnumerator2
    Rotation=&RandomRangeVector(Minimum=<-0.05 0 0>, Maximum=<0.05 0 0>)
    Use Rotation Vector=true
   )
  ]
  particleUpdateScript [
   &ParticleState()
   &ScaleMeshSize(Scale Factor=&Multiply_VectorByFloat(
     Vector=<1 1 1>
     Float=&FloatFromCurve(FloatCurve=DataInterface() {propertyValues {Curve {keys [
          {arriveTangent=-2.0161, interpMode=RCIM_Cubic, leaveTangent=-2.0161, tangentMode=RCTM_User, time=0, value=1}
          {arriveTangent=-0.3283, interpMode=RCIM_Cubic, leaveTangent=-0.3283, tangentMode=RCTM_User, time=0.6, value=0.2}
         ]}}})
    ))
   /SetVariables(Particles.Scale002:Vector3f=&Multiply_VectorByFloat(
     Vector=<1 1 1>
     Float=&FloatFromCurve(
      FloatCurve=DataInterface() {propertyValues {Curve {keys [
          {arriveTangent=-1.1673, interpMode=RCIM_Cubic, leaveTangent=-1.1673, tangentMode=RCTM_User, time=0, value=1}
          {arriveTangent=-0.3063, interpMode=RCIM_Cubic, leaveTangent=-0.3063, tangentMode=RCTM_User, time=1, value=0.3}
         ]}}}
      Scale Curve=1.15
     )
    )) {isSetParameters=true}
   &SolveForcesAndVelocity(Clamp Velocity=false, Limit Acceleration=false)
   &ScaleColor(
    Scale Alpha=&FloatFromCurve(FloatCurve=DataInterface() {propertyValues {Curve {keys [
         {arriveTangent=0.0341, interpMode=RCIM_Cubic, leaveTangent=0.0341, tangentMode=RCTM_User, time=0, value=0}
         {arriveTangent=0.0847, interpMode=RCIM_Cubic, leaveTangent=-0.0684, tangentMode=RCTM_Break, time=0.4, value=1}
         {arriveTangent=0.0005, interpMode=RCIM_Cubic, leaveTangent=0.0005, tangentMode=RCTM_User, time=0.7, value=0}
        ]}}})
    Scale RGB=&MakeVectorFromLinearColorRGB(Color=<rgba 0.3782 0.0729 1 1>)
    ScaleA=true
    ScaleRGB=true
   )
   /SetVariables(Particles.Color002=&MakeLinearColorFromVectorAndFloat(
     "Vector (RGB)"=&MakeVectorFromLinearColorRGB(Color=<rgba 0.5126 0.0659 1.5 1>)
     "Float (Alpha)"=&FloatFromCurve(
      FloatCurve=DataInterface() {propertyValues {Curve {keys [
          {arriveTangent=0.0068, interpMode=RCIM_Cubic, leaveTangent=0.0068, tangentMode=RCTM_User, time=0, value=0}
          {arriveTangent=-0.018, interpMode=RCIM_Cubic, leaveTangent=-0.018, tangentMode=RCTM_User, time=0.2, value=1}
          {arriveTangent=0.0005, interpMode=RCIM_Cubic, leaveTangent=0.0005, tangentMode=RCTM_User, time=0.5, value=0}
         ]}}}
      Scale Curve=&Multiply_Float(A=0.5, B=&FNTimeOfDayScalar(Night Scale=0.8, Morning Scale=1, Evening Scale=1))
     )
    )) {isSetParameters=true}
  ]
  rendererClasses [&NiagaraMeshRendererProperties]
  renderers [
   {rendererClass=&NiagaraMeshRendererProperties, rendererIndex=0}
   {rendererClass=&NiagaraMeshRendererProperties, rendererIndex=1}
  ]
 }
 Glow {
  emitterUpdateScript [
   &EmitterState(
    Inactive Response=::NewEnumerator1
    Life Cycle Mode=::NewEnumerator1
    Loop Behavior=::NewEnumerator1
    Loop Delay=1.6
    Loop Duration=2
    UseLoopDelay=false
   )
   &SpawnBurst_Instantaneous(Use Spawn Probability=false, UseLoopCountLimit=false)
  ]
  particleSpawnScript [
   &InitializeParticle(
    Color=<rgba 0.992 0.1042 2.5 1>
    Position Offset=$System.Position
    Sprite Size Mode=::NewEnumerator3
    Uniform Sprite Size=160
    UsePositionOffset=true
    Write Material Random=true
   )
   &CameraOffset(Camera Offset Amount=80) {enabled=false}
  ]
  particleUpdateScript [
   &ParticleState()
   &ScaleSpriteSize(
    Uniform Curve Scale=&FN_FrontEndFloatModifier(Gameplay=1, Front End=0.8)
    Uniform Curve Sprite Scale=DataInterface() {propertyValues {Curve {keys [
        {arriveTangent=0.4801, interpMode=RCIM_Cubic, leaveTangent=0.4801, tangentMode=RCTM_User, time=0, value=0}
        {arriveTangent=0.9555, interpMode=RCIM_Cubic, leaveTangent=5.5594, tangentMode=RCTM_Break, time=0.5, value=0.35}
        {arriveTangent=3.9949, interpMode=RCIM_Cubic, leaveTangent=3.9949, tangentMode=RCTM_User, time=0.6475, value=1.6025}
        {arriveTangent=0.0597, interpMode=RCIM_Cubic, leaveTangent=0.0597, tangentMode=RCTM_User, time=1, value=2.5}
       ]}}}
   )
   &ScaleColor(
    Scale Alpha=&FloatFromCurve(
     FloatCurve=DataInterface() {propertyValues {Curve {keys [
         {arriveTangent=0.2601, interpMode=RCIM_Cubic, leaveTangent=0.2601, tangentMode=RCTM_User, time=0, value=0}
         {arriveTangent=1.7089, interpMode=RCIM_Cubic, leaveTangent=3.0471, tangentMode=RCTM_Break, time=0.5, value=0.4}
         {arriveTangent=-0.0028, interpMode=RCIM_Cubic, leaveTangent=-3.936, tangentMode=RCTM_Break, time=0.6, value=1.5}
         {arriveTangent=-3.2198, interpMode=RCIM_Cubic, leaveTangent=-3.2198, tangentMode=RCTM_User, time=0.7, value=0.35}
         {arriveTangent=0.0005, interpMode=RCIM_Cubic, leaveTangent=0.0005, tangentMode=RCTM_User, time=1, value=0}
        ]}}}
     Scale Curve=&Multiply_Float(A=2, B=&FNTimeOfDayScalar(Night Scale=0.5, Morning Scale=0.7, Evening Scale=0.7))
    )
    Scale RGB=&MakeVectorFromLinearColorRGB(Color=<rgba 1 1 1 1>)
    ScaleA=true
    ScaleRGB=true
   )
   &FN_Light_Attributes(
    Color=&ColorFromCurve(ColorCurve=DataInterface() {propertyValues {
       AlphaCurve {keys []}
       BlueCurve {keys [<at 0.5055 3>, {arriveTangent=-1.3853, interpMode=RCIM_Cubic, time=0.6, value=2}, <at 0.9742 2 RCIM_Cubic>]}
       GreenCurve {keys [<at 0.5055 0.2312>, {arriveTangent=-0.1456, interpMode=RCIM_Cubic, time=0.6, value=0.3517}, <at 0.9742 0.1262 RCIM_Cubic>]}
       RedCurve {keys [<at 0.5055 0.6705>, {arriveTangent=-0.6918, interpMode=RCIM_Cubic, time=0.6, value=0.6065}, <at 0.9742 0.1711 RCIM_Cubic>]}
      }})
    Create Color Attribute=true
    Create Exponent Attribute=true
    Create Position Attribute=true
    Create Radius Attribute=true
    Enable Particle Light Enable=true
    Exponent=2
    Position=<0 0 -15>
    Radius=&Multiply_Float(
     A=200
     B=&FloatFromCurve(FloatCurve=DataInterface() {propertyValues {Curve {keys [
          <at 0.2 0.5>
          {arriveTangent=2.717, interpMode=RCIM_Cubic, leaveTangent=2.717, tangentMode=RCTM_User, time=0.5, value=0.5}
          {arriveTangent=1, interpMode=RCIM_Cubic, leaveTangent=1, time=0.6441, value=0.8052}
          {arriveTangent=0.1273, interpMode=RCIM_Cubic, leaveTangent=0.1273, tangentMode=RCTM_User, time=1, value=1}
         ]}}})
    )
    Scale Color=&FloatFromCurve(
     FloatCurve=DataInterface() {propertyValues {Curve {keys [
         {arriveTangent=1.0671, time=0, value=0}
         {arriveTangent=0.5833, leaveTangent=0.5833, time=0.2, value=0.6}
         {arriveTangent=1.4, leaveTangent=1.4, time=0.5, value=0.6}
         {arriveTangent=-7.0331, interpMode=RCIM_Cubic, leaveTangent=-7.0331, tangentMode=RCTM_User, time=0.6, value=1}
         {arriveTangent=-2.5, interpMode=RCIM_Cubic, leaveTangent=-2.5, time=0.7, value=0.4}
         {arriveTangent=-0.0943, interpMode=RCIM_Cubic, leaveTangent=-0.0943, tangentMode=RCTM_User, time=1, value=0}
        ]}}}
     Scale Curve=&Multiply_Float(A=30, B=&FNTimeOfDayScalar(Night Scale=0.5, Morning Scale=0.8, Evening Scale=0.8))
    )
   )
  ]
  rendererClasses [&NiagaraSpriteRendererProperties, &NiagaraLightRendererProperties]
  renderers [
   {rendererClass=&NiagaraSpriteRendererProperties}
   {rendererClass=&NiagaraLightRendererProperties, rendererIndex=1}
  ]
 }
 Particle {
  emitterUpdateScript [
   &EmitterState(Inactive Response=::NewEnumerator1, Life Cycle Mode=::NewEnumerator1, Loop Behavior=::NewEnumerator1, Loop Duration=1, UseLoopDelay=false)
   &SpawnBurst_Instantaneous(Spawn Count=10, Use Spawn Probability=false, UseLoopCountLimit=false)
   &SpawnRate(
    SpawnRate=&FloatFromCurve(
     FloatCurve=DataInterface() {propertyValues {Curve {keys [<at 0 1>,<at 0.2 1>,<at 0.3 2>,<at 0.4 0>,<at 1 0>]}}}
     CurveIndex=$Emitter.Age
     Scale Curve=80
    )
    Use Spawn Probability=false
   )
  ]
  particleSpawnScript [
   &InitializeParticle(
    Color=<rgba 1.8703 0 10 1>
    Lifetime Max=0.3
    Lifetime Min=0.2
    Lifetime Mode=::NewEnumerator1
    Sprite Size Max=<2.5 4>
    Sprite Size Min=<1.5 3>
    Sprite Size Mode=::NewEnumerator2
    UsePositionOffset=true
    Write Material Random=true
   )
   &ShapeLocation(
    MaskSpawnGroup=false
    Offset=$System.Position
    Offset Mode=::NewEnumerator0
    Override Seed=false
    Rotation Mode=::NewEnumerator1
    Sphere Radius=66
    UseOwnerScale=true
   )
   &AddVelocity(
    Module.Use Falloff Distance=false
    Origin Offset=$System.Position
    Override Seed=false
    Velocity Mode=::NewEnumerator1
    Velocity Speed=&FloatFromCurve(
     FloatCurve=DataInterface() {propertyValues {Curve {keys [<at 0 1>,<at 0.7 1>,<at 1 0>]}}}
     CurveIndex=$Emitter.Age
     Scale Curve=&RandomRangeFloat(Minimum=-50, Maximum=-70)
    )
   )
  ]
  particleUpdateScript [
   &ParticleState()
   &ScaleSpriteSize(Uniform Curve Sprite Scale=DataInterface() {propertyValues {Curve {keys [
        <at 0 0>
        {arriveTangent=1.5, leaveTangent=1.5, time=0.0854, value=0.85}
        <at 0.2705 0.0312>
        <at 0.3267 0.9335>
        {arriveTangent=-0.25, leaveTangent=-0.25, time=0.4998, value=0.1033}
        {arriveTangent=-0.5, leaveTangent=-0.5, time=0.6227, value=0.9345}
        <at 0.8 0.05>
        <at 0.9 0.9>
        <at 1 0>
       ]}}})
   &CurlNoiseForce(Noise Strength=100, Pan Noise=true, Pan Noise Field=<0 0.2 0.2>)
   &SolveForcesAndVelocity(Clamp Velocity=false, Limit Acceleration=false)
   &ScaleColor(
    Scale Alpha=&FloatFromCurve(
     FloatCurve=DataInterface() {propertyValues {Curve {keys [
         {interpMode=RCIM_Cubic, tangentMode=RCTM_User, time=0, value=0}
         {interpMode=RCIM_Cubic, tangentMode=RCTM_User, time=0.5, value=1}
         {tangentMode=RCTM_User, time=1, value=0}
        ]}}}
     Scale Curve=&MakeCustomFloatFromBool(Bool=$User.bIsFrontEndPreview, True Float=0.5, False Float=1)
    )
    ScaleA=true
    ScaleRGB=true
   )
   &FN_Sprite_Attributes(
    Color=&MultiplyLinearColor(
     Linear Color=<rgba 0.2963 0.097 1 1>
     Scale Factor=&MakeVector4_FromFloat(Float=&FloatFromCurve(
       FloatCurve=DataInterface() {propertyValues {Curve {keys [
           {interpMode=RCIM_Cubic, tangentMode=RCTM_User, time=0, value=0}
           {interpMode=RCIM_Cubic, tangentMode=RCTM_User, time=0.5, value=0.8}
           {tangentMode=RCTM_User, time=1, value=0}
          ]}}}
       Scale Curve=&MakeCustomFloatFromBool(Bool=$User.bIsFrontEndPreview, True Float=0.5, False Float=1)
      ))
    )
    Create Color Attribute=true
    Create SpriteSize Attribute=true
    SpriteSize=&Multiply_Vector2D(
     A=&RandomRangeVector2D(Minimum=<3 5>, Maximum=<4 6>)
     B=&Vector2DFromFloat(Value=&FloatFromCurve(
       FloatCurve=DataInterface() {propertyValues {Curve {keys [<at 0 0>,<at 0.1 0.9>,<at 0.2 0.35>,<at 0.3 1>,<at 0.4 0.15>,<at 0.5 0.9>,<at 0.6 0.5>,<at 0.7 0.9>,<at 0.8 0.35>,<at 0.9 0.95>,<at 1 0>]}}}
       Scale Curve=&RandomRangeFloat(Minimum=3, Maximum=4)
      ))
    )
    Write to Module Namespace=true
   )
  ]
  rendererClasses [&NiagaraSpriteRendererProperties]
  renderers [
   {rendererClass=&NiagaraSpriteRendererProperties}
   {rendererClass=&NiagaraSpriteRendererProperties, rendererIndex=1}
  ]
 }
 ParticleDecale {
  emitterUpdateScript [
   &EmitterState(Inactive Response=::NewEnumerator1, Life Cycle Mode=::NewEnumerator1, Loop Behavior=::NewEnumerator1, Loop Duration=2, UseLoopDelay=true)
   &SpawnBurst_Instantaneous(Spawn Count=12, Use Spawn Probability=false, UseLoopCountLimit=false)
  ]
  particleSpawnScript [
   &FN_SetSpawnGroup(Number of Spawn Groups=12)
   &InitializeParticle(
    Color=<rgba 2.2 0 6 1>
    Lifetime=1.5
    Sprite Size Mode=::NewEnumerator3
    Uniform Sprite Size=&SelectFloatFromArray(
     Direct Array Index=$Particles.SpawnGroup
     Float Selection Array=DataInterface() {propertyValues {FloatData [60,50,50,50,50,50,50,100,100,50,50,50]}}
    )
    UsePositionOffset=false
    Write Material Random=true
   )
   &CameraOffset(Camera Offset Amount=&SelectFloatFromArray(
     Direct Array Index=$Particles.SpawnGroup
     Float Selection Array=DataInterface() {propertyValues {FloatData [30,30,30,30,30,30,30,40,40,40,40,30]}}
    ))
  ]
  particleUpdateScript [
   &ParticleState()
   &SkeletalMeshLocation(
    Bone / Socket Index=$Particles.SpawnGroup
    Bone Sampling Mode=::NewEnumerator4
    MaskSpawnGroup=false
    Skeletal Mesh=DataInterface() {propertyValues {
      DefaultMesh=null
      ExcludeBoneName=root
      OptionalComponentClass=null
      PreviewMesh=/BRCosmetics/Characters/Player/Female/Medium/Bodies/F_MED_SnapFreeze_Graze/Meshes/F_MED_SnapFreeze_Graze
      SkinningMode=SkinOnTheFly
      SoftSourceActor=null
      SourceMode=Default
      UvSetIndex=0
      WholeMeshLOD=-1
      bExcludeBone=false
      bReadDeformedGeometry=true
      bRequireCurrentFrameData=true
      ComponentTags []
      FilteredBones [pelvis,upperarm_l,upperarm_r,lowerarm_l,lowerarm_r,hand_l,hand_r,calf_l,calf_r,foot_l,foot_r,head]
      FilteredSockets []
      MeshUserParameter {parameter {name=None, typeDefHandle {registeredTypeIndex=1}, varData []}}
      SamplingRegions []
     }}
   )
   &ScaleSpriteSize(Uniform Curve Sprite Scale=DataInterface() {propertyValues {Curve {keys [
        <at 0 1 RCIM_Cubic>
        {arriveTangent=0.0106, interpMode=RCIM_Cubic, leaveTangent=0.6207, tangentMode=RCTM_Break, time=0.6, value=1}
        <at 1 1.2 RCIM_Cubic>
       ]}}})
   &ScaleColor(
    Scale Alpha=&FloatFromCurve(FloatCurve=DataInterface() {propertyValues {Curve {keys [
         <at 0 0>
         <at 0.3 0.17>
         {arriveTangent=-0.6021, interpMode=RCIM_Cubic, leaveTangent=-1.1032, tangentMode=RCTM_Break, time=0.6, value=0.17}
         <at 1 0>
        ]}}})
    Scale RGB=&VectorFromFloat(Value=&FNTimeOfDayScalar(Night Scale=1.5, Morning Scale=1, Evening Scale=1))
    ScaleA=true
    ScaleRGB=true
   )
  ]
  rendererClasses [&NiagaraSpriteRendererProperties]
  renderers [
   {rendererClass=&NiagaraSpriteRendererProperties}
   {rendererClass=&NiagaraSpriteRendererProperties, rendererIndex=1}
  ]
 }
 ParticleDecale003 {
  emitterUpdateScript [
   &EmitterState(
    Inactive Response=::NewEnumerator1
    Life Cycle Mode=::NewEnumerator1
    Loop Behavior=::NewEnumerator1
    Loop Delay=0.5
    Loop Duration=2
    UseLoopDelay=true
   )
   &SpawnBurst_Instantaneous(Use Spawn Probability=false, UseLoopCountLimit=false)
   &SpawnBurst_Instantaneous(Spawn Time=0.04, Use Spawn Probability=false, UseLoopCountLimit=false) {module=SpawnBurst_Instantaneous001}
   &SpawnBurst_Instantaneous(Spawn Time=0.08, Use Spawn Probability=false, UseLoopCountLimit=false) {module=SpawnBurst_Instantaneous002}
  ]
  particleSpawnScript [
   &InitializeParticle(
    Color=<rgba 3.6667 0 10 1>
    Lifetime=0.2
    Sprite Size Mode=::NewEnumerator3
    Uniform Sprite Size=130
    UsePositionOffset=false
    Write Material Random=true
   )
   &CameraOffset(Camera Offset Amount=30)
  ]
  particleUpdateScript [
   &ParticleState()
   &ShapeLocation(
    MaskSpawnGroup=false
    Offset=&MakeVector(X=0, Y=0, Z=&Lerp_Float(A=20, B=-100, Alpha=$Particles.NormalizedAge))
    Offset Mode=::NewEnumerator0
    Override Seed=false
    Sphere Radius=1
    UseOwnerScale=true
   )
   &ScaleSpriteSize(Uniform Curve Sprite Scale=DataInterface() {propertyValues {Curve {keys [{tangentMode=RCTM_User, time=0.9, value=1}, {tangentMode=RCTM_User, time=1, value=0}]}}})
   &ScaleColor(
    Scale Alpha=&FloatFromCurve(
     FloatCurve=DataInterface() {propertyValues {Curve {keys [
         <at 0.5 1 RCIM_Cubic>
         {arriveTangent=-1.3121, interpMode=RCIM_Cubic, leaveTangent=-1.3121, tangentMode=RCTM_User, time=0.8, value=0.8}
         {arriveTangent=-4.9417, interpMode=RCIM_Cubic, leaveTangent=-4.9417, tangentMode=RCTM_User, time=1, value=0}
        ]}}}
     Scale Curve=&FN_FrontEndFloatModifier(Gameplay=1, Front End=0.5)
    )
    Scale RGB=&VectorFromFloat(Value=&FNTimeOfDayScalar(Night Scale=1.2, Morning Scale=1, Evening Scale=1))
    ScaleA=true
    ScaleRGB=true
   )
  ]
  rendererClasses [&NiagaraSpriteRendererProperties]
  renderers [
   {rendererClass=&NiagaraSpriteRendererProperties}
   {rendererClass=&NiagaraSpriteRendererProperties, rendererIndex=1}
  ]
 }
 Particle_Line {
  emitterUpdateScript [
   &EmitterState(Inactive Response=::NewEnumerator1, Life Cycle Mode=::NewEnumerator1, Loop Behavior=::NewEnumerator1, Loop Duration=0.4, UseLoopDelay=false)
   &SpawnRate(
    SpawnRate=&FloatFromCurve(
     FloatCurve=DataInterface() {propertyValues {Curve {keys [
         <at 0 1 RCIM_Cubic>
         {arriveTangent=-0.028, interpMode=RCIM_Cubic, leaveTangent=-0.028, tangentMode=RCTM_User, time=0.3, value=2.5}
         <at 1 0 RCIM_Cubic>
        ]}}}
     CurveIndex=$Emitter.NormalizedLoopAge
     Scale Curve=35
    )
    Use Spawn Probability=false
   )
  ]
  particleSpawnScript [
   &InitializeParticle(
    Color Maximum=<rgba 6.1318 0.625 20 1>
    Color Minimum=<rgba 0.0573 0 1 1>
    Color Mode=::NewEnumerator2
    Lifetime Max=0.5
    Lifetime Min=0.3
    Lifetime Mode=::NewEnumerator1
    Sprite Size Max=<2 50>
    Sprite Size Min=<1.5 20>
    Sprite Size Mode=::NewEnumerator2
    UsePositionOffset=true
    Write Material Random=true
   )
   &ShapeLocation(
    Hemisphere Distribution=<0.5 1>
    MaskSpawnGroup=false
    Offset=$System.Position
    Offset Mode=::NewEnumerator0
    Override Seed=false
    Sphere Radius=80
    Sphere Surface Distribution=1
    UseOwnerScale=true
   )
   &AddVelocity(
    Module.Use Falloff Distance=false
    Origin Offset=$System.Position
    Override Seed=false
    Velocity Mode=::NewEnumerator1
    Velocity Speed=&RandomRangeFloat(Minimum=-200, Maximum=-300)
   )
  ]
  particleUpdateScript [
   &ParticleState()
   &ScaleSpriteSize(Uniform Curve Sprite Scale=DataInterface() {propertyValues {Curve {keys [
        {arriveTangent=2.1749, interpMode=RCIM_Cubic, leaveTangent=2.1749, tangentMode=RCTM_User, time=0, value=0}
        {
         arriveTangent=0.5496
         arriveTangentWeight=0.232
         interpMode=RCIM_Cubic
         leaveTangent=0.5496
         leaveTangentWeight=2.2538
         tangentMode=RCTM_User
         time=0.4
         value=0.75
        }
        {arriveTangent=0.1525, interpMode=RCIM_Cubic, leaveTangent=0.1525, tangentMode=RCTM_User, time=1, value=1}
       ]}}})
   &Drag(Drag=&RandomRangeFloat(Minimum=2, Maximum=4), Use Linear Drag=true, Use Rotational Drag=true)
   &SolveForcesAndVelocity(Clamp Velocity=false, Limit Acceleration=false)
   &ScaleColor(
    Scale Alpha=&Lerp_Float(
     A=0
     B=1
     Alpha=&FloatFromCurve(FloatCurve=DataInterface() {propertyValues {Curve {keys [
          {interpMode=RCIM_Cubic, tangentMode=RCTM_User, time=0, value=1}
          {interpMode=RCIM_Cubic, tangentMode=RCTM_User, time=0.5, value=1}
          {tangentMode=RCTM_User, time=1, value=0}
         ]}}})
    )
    Scale RGB=&VectorFromFloat(Value=&FloatFromCurve(
      FloatCurve=DataInterface() {propertyValues {Curve {keys [
          {arriveTangent=-2.3805, interpMode=RCIM_Cubic, leaveTangent=-2.3805, tangentMode=RCTM_User, time=0, value=1}
          {interpMode=RCIM_Cubic, tangentMode=RCTM_User, time=1, value=0}
         ]}}}
      Scale Curve=&RandomRangeFloat(Minimum=1, Maximum=1.5)
     ))
    ScaleA=true
    ScaleRGB=true
   )
  ]
  rendererClasses [&NiagaraSpriteRendererProperties]
  renderers [{rendererClass=&NiagaraSpriteRendererProperties}]
 }
 Refraction {
  emitterUpdateScript [
   &EmitterState(Inactive Response=::NewEnumerator1, Life Cycle Mode=::NewEnumerator1, Loop Behavior=::NewEnumerator1, Loop Duration=2, UseLoopDelay=false)
   &SpawnBurst_Instantaneous(Spawn Count=1, Use Spawn Probability=true)
  ]
  particleSpawnScript [&InitializeParticle(
    Mesh Scale Mode=::NewEnumerator3
    Mesh Uniform Scale=1.5
    Position Offset=$System.Position
    UsePositionOffset=true
    Write Material Random=true
   ) {module=InitializeParticle001}]
  particleUpdateScript [
   &ParticleState()
   &ScaleColor(
    Scale Alpha=&FloatFromCurve(FloatCurve=DataInterface() {propertyValues {Curve {keys [<at 0 1>,<at 0.5 1>,<at 1 0>]}}})
    Scale RGB=&MakeVectorFromLinearColorRGB(Color=<rgba 1 1 1 1>)
    ScaleA=true
    ScaleRGB=true
   )
   &ScaleMeshSize(Scale Factor=&VectorFromFloat(Value=&FloatFromCurve(
      FloatCurve=DataInterface() {propertyValues {Curve {keys [
          {arriveTangent=-0.7997, interpMode=RCIM_Cubic, leaveTangent=-0.7997, tangentMode=RCTM_User, time=0, value=1}
          {arriveTangent=-2.6181, interpMode=RCIM_Cubic, leaveTangent=2.7111, tangentMode=RCTM_Break, time=0.4, value=0}
          {arriveTangent=0.29, interpMode=RCIM_Cubic, leaveTangent=0.29, tangentMode=RCTM_User, time=1, value=0.9}
         ]}}}
      Scale Curve=&FN_FrontEndFloatModifier(Gameplay=1, Front End=0.7)
     )))
   &CameraOffset(Camera Offset Amount=&FloatFromCurve(FloatCurve=DataInterface() {propertyValues {Curve {keys [<at 0 -0.5>,<at 1 1>]}}}, Scale Curve=50))
   &DynamicMaterialParameters(
    Index 0 Param 3=&FloatFromCurve(FloatCurve=DataInterface() {propertyValues {Curve {keys [<at 0.3 1>,<at 0.5 0.75>,<at 0.7 0.4>,<at 1 1>]}}})
    Index 0 Param 4=&FloatFromCurve(
     FloatCurve=DataInterface() {propertyValues {Curve {keys [
         {arriveTangent=3.8831, interpMode=RCIM_Cubic, leaveTangent=3.8831, tangentMode=RCTM_User, time=0.6, value=0}
         {arriveTangent=0.0986, interpMode=RCIM_Cubic, leaveTangent=0.0986, tangentMode=RCTM_User, time=1, value=1}
        ]}}}
     Scale Curve=20
    )
    Module.Index0Param0Write=true
    Module.Index0Param1Write=true
    Module.Index0Param2Write=true
    Module.Index0Param3Write=true
   )
  ]
  rendererClasses [&NiagaraMeshRendererProperties]
  renderers [{rendererClass=&NiagaraMeshRendererProperties, rendererIndex=0}]
 }
 Sweep_glow {
  emitterUpdateScript [
   &EmitterState(
    Inactive Response=::NewEnumerator1
    Life Cycle Mode=::NewEnumerator1
    Loop Behavior=::NewEnumerator1
    Loop Delay=0.5
    Loop Duration=1
    UseLoopDelay=true
   )
   &SpawnBurst_Instantaneous(Use Spawn Probability=false, UseLoopCountLimit=false)
  ]
  particleSpawnScript [
   &InitializeParticle(
    Color=<rgba 0.5777 0.1772 2.5 1>
    Lifetime=0.2
    Position Offset=<0 0 30>
    Sprite Size=<650 450>
    Sprite Size Mode=::NewEnumerator4
    UsePositionOffset=true
    Write Material Random=true
   )
   &CameraOffset(Camera Offset Amount=100)
   &AddVelocity(Velocity=<0 0 -800>)
  ]
  particleUpdateScript [
   &ParticleState()
   &ScaleSpriteSize(
    Uniform Curve Scale=&FN_FrontEndFloatModifier(Gameplay=1, Front End=0.8)
    Uniform Curve Sprite Scale=DataInterface() {propertyValues {Curve {keys [<at 0 0.55 RCIM_Cubic>, {arriveTangent=-0.05, interpMode=RCIM_Cubic, leaveTangent=-0.05, time=0.5, value=1}, <at 1 0.5>]}}}
   )
   &ScaleColor(
    Scale Alpha=&FloatFromCurve(
     FloatCurve=DataInterface() {propertyValues {Curve {keys [
         {interpMode=RCIM_Cubic, tangentMode=RCTM_User, time=0, value=0}
         {interpMode=RCIM_Cubic, tangentMode=RCTM_User, time=0.5, value=1}
         {arriveTangent=-2.4352, interpMode=RCIM_Cubic, leaveTangent=-2.4352, tangentMode=RCTM_User, time=1, value=0}
        ]}}}
     Scale Curve=&Multiply_Float(A=0.1, B=0.25)
    )
    ScaleA=true
    ScaleRGB=true
   )
   &Drag(Drag=2, Use Linear Drag=true, Use Rotational Drag=true)
   &SolveForcesAndVelocity(Clamp Velocity=false, Limit Acceleration=false)
  ]
  rendererClasses [&NiagaraSpriteRendererProperties]
  renderers [{rendererClass=&NiagaraSpriteRendererProperties}]
 }
 Wave {
  emitterUpdateScript [
   &EmitterState(Inactive Response=::NewEnumerator1, Life Cycle Mode=::NewEnumerator1, Loop Behavior=::NewEnumerator1, Loop Duration=2, UseLoopDelay=true)
   &SpawnBurst_Instantaneous(Use Spawn Probability=false, UseLoopCountLimit=false)
  ]
  particleSpawnScript [&InitializeParticle(
    Color=<rgba 1.5 1.5 1.5 1>
    Lifetime=0.8
    Position Offset=$System.Position
    Sprite Rotation Angle=90
    Sprite Rotation Mode=::NewEnumerator0
    Sprite Size Mode=::NewEnumerator3
    Uniform Sprite Size=130
    UsePositionOffset=true
    Write Material Random=true
   )]
  particleUpdateScript [
   &ParticleState()
   &ScaleSpriteSize(
    Uniform Curve Scale=2
    Uniform Curve Sprite Scale=DataInterface() {propertyValues {Curve {keys [
        {arriveTangent=-1.6271, interpMode=RCIM_Cubic, leaveTangent=-1.6271, tangentMode=RCTM_User, time=0, value=1}
        {arriveTangent=-0.202, interpMode=RCIM_Cubic, leaveTangent=-0.202, tangentMode=RCTM_User, time=1, value=0.2}
       ]}}}
   )
   &DynamicMaterialParameters(
    Index 0 Param 1=1
    Index 0 Param 2=&FloatFromCurve(FloatCurve=DataInterface() {propertyValues {Curve {keys [
         {interpMode=RCIM_Cubic, tangentMode=RCTM_User, time=0, value=1}
         {arriveTangent=-1, interpMode=RCIM_Cubic, leaveTangent=-1, time=0.6, value=0.7}
         {interpMode=RCIM_Cubic, leaveTangent=-3.3753, time=1, value=0}
        ]}}})
    Index 0 Param 3=1
    Index 0 Param 4=1
    Module.Param0WriteEnabled=true
    Module.Param1WriteEnabled=true
    Module.Param2WriteEnabled=true
    Module.Param3WriteEnabled=true
   )
   &ScaleColor(
    Scale Alpha=&FloatFromCurve(FloatCurve=DataInterface() {propertyValues {Curve {keys [
         {arriveTangent=-0.069, interpMode=RCIM_Cubic, leaveTangent=-0.069, tangentMode=RCTM_User, time=0, value=0}
         {arriveTangent=0.0091, interpMode=RCIM_Cubic, leaveTangent=0.0091, tangentMode=RCTM_User, time=0.4, value=1}
         {arriveTangent=-1.2187, interpMode=RCIM_Cubic, leaveTangent=-1.2187, tangentMode=RCTM_User, time=0.6, value=0.2}
         {interpMode=RCIM_Cubic, tangentMode=RCTM_User, time=1, value=0}
        ]}}})
    Scale RGB=&MakeVectorFromLinearColorRGB(Color=&MultiplyLinearColor(
      Linear Color=&ColorFromCurve(ColorCurve=DataInterface() {propertyValues {
         AlphaCurve {keys []}
         BlueCurve {keys [<at 0.0963 2 RCIM_Cubic>, <at 0.2351 5>, {interpMode=RCIM_Cubic, leaveTangent=-0.7175, time=0.4949, value=2}]}
         GreenCurve {keys [<at 0.0963 0.1681 RCIM_Cubic>, <at 0.2351 0.2296>, {interpMode=RCIM_Cubic, leaveTangent=-4.4353, time=0.4949, value=0}]}
         RedCurve {keys [<at 0.0963 0.5589 RCIM_Cubic>, <at 0.2351 1.7644>, {interpMode=RCIM_Cubic, leaveTangent=2.2176, time=0.4949, value=0.3}]}
        }})
      Scale Factor=<1 1 1 1>
     ))
    ScaleA=true
    ScaleRGB=true
   )
   /SetVariables(Particles.Color02:LinearColor=&MakeLinearColorFromVectorAndFloat(
     "Vector (RGB)"=&MakeVectorFromLinearColorRGB(Color=&MultiplyLinearColor(
       Linear Color=<rgba 1.7071 0.6284 5 1>
       Scale Factor=&MakeVector4_FromFloat(Float=&FN_PickFloatDependOnPlatform(TargetPlatformActive=1, TargetPlatformNoActive=0.5))
      ))
     "Float (Alpha)"=&FloatFromCurve(
      FloatCurve=DataInterface() {propertyValues {Curve {keys [
          {arriveTangent=0.0195, interpMode=RCIM_Cubic, leaveTangent=0.0195, tangentMode=RCTM_User, time=0, value=0}
          <at 0.3 1 RCIM_Cubic>
          {interpMode=RCIM_Cubic, tangentMode=RCTM_User, time=1, value=0}
         ]}}}
      Scale Curve=0.8
     )
    )) {isSetParameters=true, module=SetVariables_#2}
  ]
  rendererClasses [&NiagaraSpriteRendererProperties]
  renderers [{rendererClass=&NiagaraSpriteRendererProperties}]
 }
}
summary {
 systemName=NS_Emote_SnapFreeze_ShockWave_Return
 emitters [
  {emitterName=Refraction, rendererClasses [&NiagaraMeshRendererProperties]}
  {emitterName=Glow, rendererClasses [&NiagaraSpriteRendererProperties, &NiagaraLightRendererProperties]}
  {emitterName=Energy_Ring, rendererClasses [&NiagaraMeshRendererProperties]}
  {emitterName=ParticleDecale, rendererClasses [&NiagaraSpriteRendererProperties]}
  {emitterName=Chest_Glow, rendererClasses [&NiagaraSpriteRendererProperties]}
  {emitterName=Color_wave, rendererClasses [&NiagaraSpriteRendererProperties]}
  {emitterName=Particle_Line, rendererClasses [&NiagaraSpriteRendererProperties]}
  {emitterName=Particle, rendererClasses [&NiagaraSpriteRendererProperties]}
  {emitterName=ParticleDecale003, rendererClasses [&NiagaraSpriteRendererProperties]}
  {emitterName=Wave, rendererClasses [&NiagaraSpriteRendererProperties]}
  {emitterName=Sweep_glow, rendererClasses [&NiagaraSpriteRendererProperties]}
 ]
 userVariables [
  User.bIsFrontEnd {type=NiagaraBool}
  User.bIsFrontEndPreview {type=NiagaraBool}
 ]
}
